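# Sound cue that plays the Heist "Heisting" POI music track as a gameplay music loop.
# A single looping wave player feeds the mastered heist wave straight to the output, with a raised volume and the gameplay music sound class.
# Overridden concurrency caps simultaneous instances at two and blocks new ones beyond that, and a very large max distance is set.
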SoundCue /HeistBasesGameplay/Audio/Music/Mx_Heist_Heisting_POI_Cue
root=WavePlayer_0
nodes [{id=WavePlayer_0, properties {bLooping=true, soundWaveAssetPtr=/HeistBasesGameplay/Audio/Music/wav/mx_Heist_Heisting_Mastered}}]
properties {
 bOverrideConcurrency=true
 maxDistance=2097152
 soundClassObject=/Game/Sounds/FortSoundClassesAndModes/Fort_Music_Gameplay
 volumeMultiplier=1.3
 concurrencyOverrides {maxCount=2, resolutionRule=PreventNew, voiceStealReleaseTime=0.2, volumeScaleAttackTime=0.2}
}
