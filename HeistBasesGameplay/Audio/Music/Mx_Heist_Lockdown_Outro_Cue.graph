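# Sound cue for the music outro of the Heist lockdown, playing a single mastered wave through one Wave Player.
# It routes through the gameplay music sound class with boosted volume and attenuation overrides that turn off spatialization.
# Its own concurrency override allows two instances and blocks new ones beyond that.
# It keeps playing when silent rather than stopping.
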SoundCue /HeistBasesGameplay/Audio/Music/Mx_Heist_Lockdown_Outro_Cue
root=WavePlayer_0
nodes [{id=WavePlayer_0, properties {soundWaveAssetPtr=/HeistBasesGameplay/Audio/Music/wav/mx_Heist_Lockdown_Outro_Mastered}}]
properties {
 bOverrideConcurrency=true
 maxDistance=2097152
 priority=2
 soundClassObject=/Game/Sounds/FortSoundClassesAndModes/Fort_Music_Gameplay
 virtualizationMode=PlayWhenSilent
 volumeMultiplier=1.867
 attenuationOverrides {bSpatialize=false}
 concurrencyOverrides {maxCount=2, resolutionRule=PreventNew, voiceStealReleaseTime=0.2, volumeScaleAttackTime=0.2}
 concurrencySet [null]
}
