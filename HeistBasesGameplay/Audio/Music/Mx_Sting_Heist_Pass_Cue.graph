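# Sound cue for the music sting that plays on a heist pass in the HeistBases gameplay mode.
# A random node picks one of five equally weighted pass sting wave variations.
# It routes to the music stinger sound class with a slight volume boost.
# A concurrency override allows one instance at a time and blocks new plays while one is active.
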SoundCue /HeistBasesGameplay/Audio/Music/Mx_Sting_Heist_Pass_Cue
root=Random_1
nodes [
 {id=Random_1, children [WavePlayer_0,WavePlayer_6,WavePlayer_7,WavePlayer_8,WavePlayer_9], properties {weights [1,1,1,1,1]}}
 {id=WavePlayer_0, properties {soundWaveAssetPtr=/HeistBasesGameplay/Audio/Music/wav/mx_Heist_Sting_Pass_01}}
 {id=WavePlayer_6, properties {soundWaveAssetPtr=/HeistBasesGameplay/Audio/Music/wav/mx_Heist_Sting_Pass_02}}
 {id=WavePlayer_7, properties {soundWaveAssetPtr=/HeistBasesGameplay/Audio/Music/wav/mx_Heist_Sting_Pass_03}}
 {id=WavePlayer_8, properties {soundWaveAssetPtr=/HeistBasesGameplay/Audio/Music/wav/mx_Heist_Sting_Pass_04}}
 {id=WavePlayer_9, properties {soundWaveAssetPtr=/HeistBasesGameplay/Audio/Music/wav/mx_Heist_Sting_Pass_05}}
]
properties {
 bOverrideConcurrency=true
 maxDistance=2097152
 soundClassObject=/Game/Sounds/FortSoundClassesAndModes/Fort_Music_Stinger_PSM
 volumeMultiplier=1.2
 concurrencyOverrides {maxCount=1, resolutionRule=PreventNew}
}
